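# Translucent deferred decal material for Niagara ribbons that draws the normal detail of a mud sliding trail.
# It angle-blends an intensified trail normal, an offset damage normal and two scales of world-XY noise normals, then lerps from flat using high-contrast noise.
# A dynamic InitialOffset shifts the detail patterns.
# An opacity chain multiplies particle alpha, a sine across V, that noise, two UV1 sphere masks and inverted depth fade.
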Material /Game/Effects/MudInteraction/Materials/MD_Mud_Sliding_Trail_Normal
graph {
 properties {
  bUsedWithNiagaraRibbons=true
  blendMode=BLEND_Translucent
  materialDomain=MD_DeferredDecal
  pixelDepthOffsetMode=PDOM_Legacy
  translucencyLightingMode=TLM_SurfacePerPixelLighting
  usedShadingModels=Default Lit
 }
 nodes [
  Constant() {r=0.5}
  Constant() {r=0.4}
  n1 = TextureCoordinate()
  n2 = DynamicParameter() {paramNames [InitialOffset, Param2, Param3, Param4]}
  n3 = Add(A=TextureCoordinate(), B=@n2.InitialOffset)
  n4 = AppendVector(A=Constant() {r=1}, B=Constant() {r=0.15})
  n5 = MaterialFunctionCall(
   BaseNormal=MaterialFunctionCall(
    Normal=TextureSample(UVs=@n1) {samplerType=SAMPLERTYPE_Normal, texture=/Game/Effects/MudInteraction/Textures/T_Mud_Sliding_Trail_Normal}.RGB
    Flatness=Constant() {r=-2}
   ) {materialFunction=&FlattenNormal}.Result
   AdditionalNormal=TextureSample(UVs=Multiply(A=@n3, B=@n4)) {samplerType=SAMPLERTYPE_Normal, texture=/Game/Global/Textures/Default/Tileable_Patterns/CarDamageNormal}.RGB
  ) {materialFunction=&BlendAngleCorrectedNormals}
  n6 = Multiply(A=@n2.InitialOffset) {constB=500}
  n7 = WorldPosition()
  n8 = Divide(A=ComponentMask(None=Add(A=@n6, B=@n7.XYZ)) {g=true, r=true}, B=Constant() {r=50})
  n9 = Divide(A=@n8, B=Constant() {r=5}) {constB=5}
  n10 = MaterialFunctionCall(
   BaseNormal=MaterialFunctionCall(
    Normal=TextureSample(UVs=@n8) {samplerType=SAMPLERTYPE_Normal, texture=/Game/Athena/Environments/DudeBro/T_Dudebro_Noise_N_01}.RGB
    Flatness=Constant() {r=0.8}
   ) {materialFunction=&FlattenNormal}.Result
   AdditionalNormal=MaterialFunctionCall(
    Normal=TextureSample(UVs=@n9) {samplerType=SAMPLERTYPE_Normal, texture=&T_Dudebro_Noise_N_01}.RGB
    Flatness=Constant() {r=0.5}
   ) {materialFunction=&FlattenNormal}.Result
  ) {materialFunction=&BlendAngleCorrectedNormals}
  n11 = Divide(A=@n9, B=Constant() {r=5})
  n12 = Multiply(
   A=Power(
    Base=TextureSample(UVs=@n11) {samplerType=SAMPLERTYPE_LinearColor, texture=&HighContrastNoise}.G
    Exp=ScalarParameter() {defaultValue=2}
   )
   B=Constant() {r=1}
  )
  LinearInterpolate(
   A=Constant3Vector() {constant=<rgba 0 0 1 0>}
   B=MaterialFunctionCall(
    Normal=MaterialFunctionCall(BaseNormal=@n5.Result, AdditionalNormal=@n10.Result) {materialFunction=&BlendAngleCorrectedNormals}.Result
    Flatness=Constant()
   ) {materialFunction=&FlattenNormal}.Result
   Alpha=@n12
  )
  n13 = TextureCoordinate()
  n14 = Sine(None=Multiply(A=MaterialFunctionCall(Float2=@n13) {materialFunction=&BreakOutFloat2Components}.G) {constB=0.5})
  n15 = Constant() {r=0.5}
  n16 = Multiply(A=Multiply(A=ParticleColor().A, B=Multiply(A=@n14, B=@n15)), B=ScalarParameter() {defaultValue=22, parameterName=Opacity})
  n17 = TextureCoordinate() {coordinateIndex=1}
  n18 = Constant() {r=3}
  n19 = Constant() {r=0.5}
  n20 = Constant() {r=-2}
  n21 = Multiply(
   A=SphereMask(A=@n17, B=AppendVector(A=@n18, B=@n19), Radius=Constant() {r=3}) {hardnessPercent=0}
   B=SphereMask(A=@n17, B=AppendVector(A=@n20, B=@n19), Radius=Constant() {r=3}) {hardnessPercent=0}
  )
  Multiply(A=Multiply(A=Multiply(A=@n16, B=@n12), B=Multiply(A=@n21) {constB=160}), B=OneMinus(None=DepthFade()))
 ]
}
