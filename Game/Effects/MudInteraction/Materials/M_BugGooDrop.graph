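# M_BugGooDrop is a Default Lit material for bug goo drops rendered as Niagara mesh particles.
# Its graph holds a single call to the material function MF_BugGooDrop, which contains the shading logic, with the material set to use material attributes.
# Automatic usage flagging in the editor is off, leaving Niagara mesh particles as the explicitly enabled usage.
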Material /Game/Effects/MudInteraction/Materials/M_BugGooDrop
graph {
 properties {bAutomaticallySetUsageInEditor=false, bUseMaterialAttributes=true, bUsedWithNiagaraMeshParticles=true, usedShadingModels=Default Lit}
 nodes [MaterialFunctionCall() {materialFunction=&/MF_BugGooDrop}]
}
