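# Translucent deferred decal material for mud interaction effects, built from an Asteria mud ground texture tinted by particle color.
# A tiled noise normal map is multiplied by a Normal scalar, and a Surface vector holds metallic, specular and roughness values.
# A radial sphere mask times decal lifetime opacity gives a fading round shape.
# A shading path switch zeroes that mask on forward and mobile paths.
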Material /Game/Effects/MudInteraction/Materials/M_Decal_Mud
graph {
 properties {
  bEnableResponsiveAA=true
  bOutputTranslucentVelocity=true
  bUsedWithNiagaraRibbons=true
  bUsedWithNiagaraSprites=true
  blendMode=BLEND_Translucent
  materialDomain=MD_DeferredDecal
  pixelDepthOffsetMode=PDOM_Legacy
  shadingModel=MSM_Unlit
  twoSided=true
  usedShadingModels=Unlit
 }
 nodes [
  DynamicParameter() {paramNames [Normal Strangth, Param2, Param3, Param4]}
  Multiply(A=TextureSample() {texture=/Game/Environments/Asteria/Terrain/Textures/T_Asteria_MudGround_A_D}.RGB, B=ParticleColor().RGB)
  Multiply(
   A=TextureSample() {samplerType=SAMPLERTYPE_Normal, texture=/Game/Effects/Fort_Effects/Textures/Tile/noise_flamy_Tile_NRM}.RGB
   B=ScalarParameter() {defaultValue=1, parameterName=Normal}
  )
  n1 = Multiply(
   A=SphereMask(
    A=TextureCoordinate()
    B=Constant() {r=0.5}
    Radius=ScalarParameter() {defaultValue=0.5, parameterName=Radius}
    Hardness=ScalarParameter() {defaultValue=0.5, parameterName=Hardness}
   )
   B=MaterialFunctionCall(Float2=DecalLifetimeOpacity().Opacity) {materialFunction=&BreakOutFloat2Components}.R
  )
  n2 = Constant()
  ShadingPathSwitch(Default=@n1, Deferred=@n1, Forward=@n2, Mobile=@n2)
  VectorParameter() {parameterName=Surface, channelNames {a="", b=Roughness, g=Specular, r=Metallic}, defaultValue=<rgba 0.1 0.1 0.4 0>}
 ]
}
