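# Translucent two-sided material for Niagara ribbons that draws tire tracks in mud.
# Color multiplies tinted mud ground texture by particle color, darkened by the tire-track mask, plus a fake specular from world-projected noise normals and the atmospheric light vector.
# Opacity combines particle alpha, the track mask, world-space noise and a ribbon edge fade.
# Paired depth fades, one inverted, limit visibility to near underlying surfaces for a decal-like look.
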Material /Game/Effects/MudInteraction/Materials/M_Mud_Tire_Trail
graph {
 properties {
  bUsedWithNiagaraRibbons=true
  blendMode=BLEND_Translucent
  pixelDepthOffsetMode=PDOM_Legacy
  translucencyPass=MTP_BeforeDOF
  twoSided=true
  usedShadingModels=Default Lit
 }
 nodes [
  n1 = VectorParameter() {parameterName=Color Adjustment, defaultValue=<rgba 2.5 2.5 2.5 0>}
  n2 = TextureSample() {texture=/Game/Environments/Asteria/Terrain/Textures/T_Asteria_MudGround_A_D}
  n3 = ParticleColor()
  n4 = TextureSampleParameter2D(UVs=TextureCoordinate() {coordinateIndex=1}) {parameterName=Tire Texture, texture=/Game/Effects/MudInteraction/Textures/T_Tire_Mud_Track}
  n5 = Clamp(None=Power(
    Base=StaticComponentMaskParameter(None=@n4.RGB) {defaultB=true, parameterName=Tire Decal Channel}
    Exp=ScalarParameter() {defaultValue=4, parameterName=sss}
   ) {constExponent=40})
  n6 = ComponentMask(None=TextureCoordinate()) {r=true}
  n7 = Multiply(A=OneMinus(None=@n6), B=@n6)
  n8 = ScalarParameter() {defaultValue=100, parameterName=edgeWidthMultiply}
  n9 = Clamp(None=Power(Base=Multiply(A=@n7, B=@n8) {constB=16}))
  n10 = Clamp(None=OneMinus(None=Multiply(A=@n5, B=@n9)))
  n11 = Multiply(A=ComponentMask(None=WorldPosition().XYZ) {g=true, r=true}) {constB=0.02}
  n12 = TextureSample(UVs=@n11) {samplerType=SAMPLERTYPE_Normal, texture=/Game/Effects/Fort_Effects/Textures/Tile/T_GaussianNoise_02_N}
  n13 = AtmosphericLightVector()
  Add(
   A=Multiply(A=Multiply(A=@n1.RGB, B=@n2.RGB), B=Multiply(A=@n3.RGB, B=@n10))
   B=Power(Base=DotProduct(A=@n12.RGB, B=@n13), Exp=ScalarParameter() {defaultValue=12, parameterName=Fake Specular})
  )
  n14 = Multiply(A=@n3.A, B=@n5)
  n15 = TextureSample(UVs=@n11) {texture=/Game/Effects/Fort_Effects/Textures/Tile/T_GaussianNoise_01_M}
  n16 = ComponentMask(None=TextureCoordinate()) {r=true}
  n17 = Multiply(A=OneMinus(None=@n16), B=@n16)
  n18 = Power(Base=Saturate(None=Subtract(A=@n17) {constB=0.01})) {constExponent=0.5}
  n19 = DepthFade(FadeDistance=ScalarParameter() {defaultValue=35, group=Decal, parameterName=VisibleDepthDistance}) {fadeDistanceDefault=80}
  Multiply(
   A=Multiply(A=Multiply(A=@n14, B=@n15.RGB), B=Multiply(B=@n18) {constA=3})
   B=DepthFade(Opacity=OneMinus(None=@n19), FadeDistance=ScalarParameter() {defaultValue=15, group=Decal, parameterName=DepthFade}) {fadeDistanceDefault=15}
  )
  OneMinus(None=@n4.A)
  TextureSample() {samplerType=SAMPLERTYPE_Masks, texture=/Game/Environments/Asteria/Terrain/Textures/T_Asteria_MudGround_A_S}
 ]
}
