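# Niagara system for a mud splash on the camera lens: lingering dirt sprites plus a brief spray burst, both gated by User.UseSplash.
# DirtOnScreen places sprites along a ring, coloured from a palette, and ignores lifetime; a timeline driven by owner execution state slides them down, fades and kills them.
# DirtOnScreen001 throws black, spinning SubUV sprites outward and upward under half gravity, growing and fading out.
# User.Duration and User.MudColor are declared, but neither emitter references them directly.
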NiagaraSystem /Game/Effects/MudInteraction/NS_Mud_Camera_Lens_Status_Splash
dependencies {
 usedDataInterfaces [
  &NiagaraDataInterfaceArrayColor
  &NiagaraDataInterfaceCurve
  &NiagaraDataInterfaceSpriteRendererInfo
 ]
 usedDynamicInputs [&Lerp_Float, &SetBoolByIntComparison]
 usedModules [
  &self:SystemSpawnScript.SystemScriptSource.SystemScriptGraph.NiagaraNodeAssignment_1.SetVariables_#1
  &self:SystemSpawnScript.SystemScriptSource.SystemScriptGraph.NiagaraNodeAssignment_0.SetVariables_#2
  &SystemState
 ]
}
emitters {
 DirtOnScreen {
  emitterSpawnScript [/SetVariables(Emitter.Random:NiagaraFloat=&RandomRangeFloat(Minimum=0, Maximum=1)) {isSetParameters=true}]
  emitterUpdateScript [
   &EmitterState(Life Cycle Mode=::NewEnumerator1, Loop Duration=1, UseLoopDelay=false)
   &SpawnBurst_Instantaneous(
    Spawn Count=15
    Spawn Probability=&MakeCustomFloatFromBool(Bool=$User.UseSplash, True Float=1, False Float=0)
    Use Spawn Probability=true
    UseLoopCountLimit=true
   )
   /SetVariables(Emitter.EmitterVelocity:Vector3f=$Engine.Owner.Velocity) {isSetParameters=true}
  ]
  particleSpawnScript [
   &InitializeParticle(
    Color=&SelectLinearColorFromArray(
     Array Sampling Mode=::NewEnumerator1
     Direct Array Index=0
     Color Selection Array=DataInterface() {propertyValues {ColorData [<rgba 0.02 0.0092 0.0023 1>, <rgba 0 2.015 3 1>, <rgba 1.7017 0 3 1>, <rgba 2 1.5804 0 1>]}}
     Random Seed=?
    )
    Position Offset=&RandomRangeVector(Minimum=<0 0 0>, Maximum=<1 0 0>)
    Sprite Rotation Angle=&Add_Float(
     A=$Emitter.Random
     B=&ReturnNormalizedExecIndex(Normalized Index Scale=3, Normalized Index Includes One=false)
    )
    Sprite Rotation Mode=::NewEnumerator1
    Sprite Size Mode=::NewEnumerator1
    Sprite UV Mode=::NewEnumerator3
    Uniform Sprite Size Max=60
    Uniform Sprite Size Min=55
    UsePositionOffset=false
    Write Material Random=true
   )
   &InitializeParticle(
    Color=<rgba 0.02 0.0092 0.0023 1>
    Position Offset=&RandomRangeVector(Minimum=<0 0 0>, Maximum=<1 0 0>)
    Sprite Rotation Angle=&Add_Float(
     A=$Emitter.Random
     B=&ReturnNormalizedExecIndex(Normalized Index Scale=3, Normalized Index Includes One=false)
    )
    Sprite Rotation Mode=::NewEnumerator1
    Sprite Size Mode=::NewEnumerator1
    Sprite UV Mode=::NewEnumerator3
    Uniform Sprite Size Max=60
    Uniform Sprite Size Min=55
    UsePositionOffset=false
    Write Material Random=true
   ) {enabled=false, module=InitializeParticle001}
   /SetVariables(Particles.SubImageIndex=&RandomRangeFloat(Minimum=0, Maximum=0)) {isSetParameters=true}
   &ShapeLocation(
    Apply To Particle Position=::NewEnumerator1
    MaskSpawnGroup=false
    Non Uniform Scale=<0.66 1 1>
    Offset=<0 0 10>
    Offset Coordinate Space=::Local
    Offset Mode=::NewEnumerator0
    Override Seed=false
    Radius Position:NiagaraFloat=&RandomRangeFloat(Minimum=0.98, Maximum=1)
    Ring / Disc Distribution Mode:ENiagaraSphereDistributionMode=::NewEnumerator1
    Ring Radius=110
    Rotation Mode=::NewEnumerator1
    Shape Primitive=::NewEnumerator4
    U Position=&Subtract_Float(A=&ReturnNormalizedExecIndex(Normalized Index Scale=0.4), B=0.2)
    UseOwnerScale=true
   )
  ]
  particleUpdateScript [
   &ParticleState(Kill Particles When Lifetime Has Elapsed=false)
   &Timeline(
    Play=&self:ScratchDynamicInput(SystemState:ENiagaraExecutionState=$Engine.Owner.ExecutionState)
    Timeline End=2
   )
   &KillParticles(Kill Particles=&SetBoolByFloatComparison(A=$Particles.Timeline.NormalizedTime, B=1, Comparison Type=::NewEnumerator2))
   /SetVariables(Particles.Position=&AddVectorToPosition(
     Position=Linked() {linkedVariable {name=Particles.Initial.ShapeLocation.ShapePosition, type=Vector3f}}
     Vector=&Multiply_VectorByFloat(
      Vector=<0 0 -1>
      Float=&FloatFromCurve(
       FloatCurve=DataInterface() {propertyValues {Curve {keys [
           {interpMode=RCIM_Cubic, tangentMode=RCTM_User, time=0, value=0}
           {arriveTangent=2.9163, interpMode=RCIM_Cubic, leaveTangent=2.9163, tangentMode=RCTM_User, time=1, value=1}
          ]}}}
       CurveIndex=$Particles.Timeline.NormalizedTime
       Scale Curve=20
      )
     )
    )) {isSetParameters=true}
   &ScaleColor(
    Scale Alpha=&FloatFromCurve(
     FloatCurve=DataInterface() {propertyValues {Curve {keys [{interpMode=RCIM_Cubic, tangentMode=RCTM_User, time=0, value=1}, {interpMode=RCIM_Cubic, tangentMode=RCTM_User, time=1, value=0}]}}}
     CurveIndex=$Particles.Timeline.NormalizedTime
     Scale Curve=&FloatFromCurve(
      FloatCurve=DataInterface() {propertyValues {Curve {keys [{tangentMode=RCTM_User, time=0, value=0}, {tangentMode=RCTM_User, time=0.15, value=1}]}}}
      CurveIndex=$Emitter.Age
     )
    )
    ScaleA=true
    ScaleRGB=true
   )
  ]
  rendererClasses [&NiagaraSpriteRendererProperties]
  renderers [{rendererClass=&NiagaraSpriteRendererProperties}]
 }
 DirtOnScreen001 {
  emitterUpdateScript [
   &EmitterState()
   &SpawnBurst_Instantaneous(
    Spawn Count=&RandomRangeInt(Minimum=30, Maximum=40)
    Spawn Probability=&MakeCustomFloatFromBool(Bool=$User.UseSplash, True Float=1, False Float=0)
    Use Spawn Probability=true
    UseLoopCountLimit=true
   )
  ]
  particleSpawnScript [
   &InitializeParticle(
    Color=<rgba 0 0 0 1>
    Lifetime Max=1.25
    Lifetime Mode=::NewEnumerator1
    Sprite Rotation Mode=::NewEnumerator2
    Sprite Size Mode=::NewEnumerator1
    Sprite UV Mode=::NewEnumerator3
    Uniform Sprite Size Max=75
    Uniform Sprite Size Min=50
    UsePositionOffset=false
    Write Material Random=true
   )
   &SubUVAnimation(
    "Sprite Renderer "=DataInterface() {propertyValues {SpriteRenderer=&self:DirtOnScreen_0.NiagaraSpriteRendererProperties_1}}
    SubUV Animation Mode:ENiagara_SubUVLookupModeV2=::NewEnumerator2
    UseEndFrame=false
    UseStartFrame=false
   )
   &ShapeLocation(
    Disc Coverage=0.75
    MaskSpawnGroup=false
    Non Uniform Scale=<0.8 1 1>
    Offset Coordinate Space=::Local
    Offset Mode=::NewEnumerator0
    Override Seed=false
    Ring Radius=50
    Rotation Mode=::NewEnumerator1
    Shape Primitive=::NewEnumerator4
    UseOwnerScale=true
   )
   &AddVelocity(Velocity=$Particles.ShapeLocation.ShapeVector, Velocity Speed Scale=3)
   &AddVelocity(Velocity=<0 0 1>, Velocity Speed Scale=&RandomRangeFloat(Minimum=5, Maximum=60)) {module=AddVelocity001}
  ]
  particleUpdateScript [
   &ParticleState()
   &SpriteRotationRate(Rotation Rate=&RandomRangeFloat(Minimum=-90, Maximum=90))
   &GravityForce(Gravity=&Multiply_VectorByFloat(Vector=<0 0 -980>, Float=0.5))
   &SolveForcesAndVelocity(Clamp Velocity=false, Limit Acceleration=false)
   &ScaleSpriteSize(Uniform Curve Sprite Scale=DataInterface() {propertyValues {Curve {keys [
        {arriveTangent=3.2408, interpMode=RCIM_Cubic, leaveTangent=3.2408, tangentMode=RCTM_User, time=0, value=0.15}
        {arriveTangent=0.85, interpMode=RCIM_Cubic, leaveTangent=0.85, time=0.25, value=0.6228}
        {arriveTangent=0.0224, interpMode=RCIM_Cubic, leaveTangent=0.0224, tangentMode=RCTM_User, time=1, value=1}
       ]}}})
   &ScaleColor(
    Scale Alpha=&FloatFromCurve(FloatCurve=DataInterface() {propertyValues {Curve {keys [{interpMode=RCIM_Cubic, tangentMode=RCTM_User, time=0, value=1}, {interpMode=RCIM_Cubic, tangentMode=RCTM_User, time=1, value=0}]}}})
    Scale RGB=<0 0 0>
    ScaleA=true
    ScaleRGB=false
   )
  ]
  rendererClasses [&NiagaraSpriteRendererProperties]
  renderers [{rendererClass=&NiagaraSpriteRendererProperties}]
 }
}
summary {
 systemName=NS_Mud_Camera_Lens_Status_Splash
 emitters [
  {emitterName=DirtOnScreen, rendererClasses [&NiagaraSpriteRendererProperties]}
  {emitterName=DirtOnScreen001, rendererClasses [&NiagaraSpriteRendererProperties]}
 ]
 userVariables [
  User.Duration {type=NiagaraFloat, defaultValue=20, description=""}
  User.MudColor {type=NiagaraInt32, defaultValue=2}
  User.UseSplash {type=NiagaraBool}
 ]
}
