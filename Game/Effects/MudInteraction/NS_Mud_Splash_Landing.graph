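# Niagara system for mud landing splashes: a single Footprint emitter renders short-lived mesh marks at supplied foot positions.
# Two instantaneous bursts form separate spawn groups, with direction-array flow rotation and orientation aligned to the ground normal.
# Age-driven curves flatten, spread and fade each mark through dynamic material parameters, which also carry the spawn-group mask and orientation quaternion.
# A fallback sand color curve samples time of day.
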NiagaraSystem /Game/Effects/MudInteraction/NS_Mud_Splash_Landing
dependencies {
 usedDataInterfaces [&NiagaraDataInterfaceColorCurve, &NiagaraDataInterfaceCurve]
 usedDynamicInputs []
 usedModules [
  &self:SystemSpawnScript.SystemScriptSource.SystemScriptGraph.NiagaraNodeAssignment_0.SetVariables_#1
  &SystemState
 ]
}
emitters {Footprint {
  emitterUpdateScript [
   &EmitterState()
   &SpawnBurst_Instantaneous(Use Spawn Probability=false, UseLoopCountLimit=false)
   &SpawnBurst_Instantaneous(Spawn Group=1, Use Spawn Probability=false, UseLoopCountLimit=false) {module=SpawnBurst_Instantaneous001}
  ]
  particleSpawnScript [
   &InitializeParticle(
    Lifetime=2.5
    Mesh Scale Mode=::NewEnumerator3
    Mesh Uniform Scale=0.9
    Position=&self:FootPositionDI(FootPositionArray:NiagaraDataInterfaceArrayPosition=$User.FootPositionArray)
    Position Mode=::NewEnumerator1
    UsePositionOffset=true
    Write Material Random=true
   )
   &InitialMeshOrientation(Mesh Orientation Mode=::NewEnumerator7, Use Rotation Vector=true)
   &self:FindFlowRotation(FootDirectionArray:NiagaraDataInterfaceArrayFloat3=$User.FootDirectionArray) {module=FindFlowRotation}
   /SetVariables(Particles.MeshOrientation=&MultiplyQuaternion(
     Quaternion A=&FindQuatBetweenTwoVectors(From Vector=<0 0 1>, To Vector=$User.GroundNormal)
     Quaternion B=&MakeQuatFromAxisAngle(Axis=<0 0 1>, AngleInDegrees=&Lerp_Float(A=0, B=-360, Alpha=$Particles.Rotation))
    )) {isSetParameters=true}
   /SetVariables(Particles.TintScale=1) {isSetParameters=true}
   /SetVariables(Particles.SandFallbackColor:LinearColor=&ColorFromCurve(
     ColorCurve=DataInterface() {propertyValues {
       AlphaCurve {keys [
         <at 0 0.3 RCIM_Cubic>
         <at 0.11 0.3>
         <at 0.125 0.1>
         <at 0.163 0.16>
         <at 0.35 0.2>
         <at 0.41 0.35>
         <at 0.65 0.3>
         <at 0.75 0.16>
         <at 0.833 0.15>
         <at 0.912 0.1>
         <at 0.98 0.17>
         {arriveTangent=6.5, interpMode=RCIM_Cubic, leaveTangent=6.5, time=1, value=0.3}
        ]}
       BlueCurve {keys []}
       GreenCurve {keys []}
       RedCurve {keys []}
      }}
     CurveIndex=Linked() {linkedVariable {name=NPC.FortniteNPC.TimeOfDay_0-1, type=NiagaraFloat}}
    )) {isSetParameters=true}
  ]
  particleUpdateScript [
   &ParticleState()
   /SetVariables(Particles.MarkFlat:NiagaraFloat=&ClampFloat(Float=&FloatFromCurve(
      FloatCurve=DataInterface() {propertyValues {Curve {keys [
          {interpMode=RCIM_Cubic, tangentMode=RCTM_Break, time=0, value=0}
          {arriveTangent=2.7798, interpMode=RCIM_Cubic, leaveTangent=2.7798, tangentMode=RCTM_User, time=1, value=1}
         ]}}}
      Scale Curve=&RandomRangeFloat(Minimum=1, Maximum=2)
     ))) {isSetParameters=true}
   /SetVariables(Particles.BlurMaskScale:NiagaraFloat=&Multiply_Float(
     A=&FloatFromCurve(
      FloatCurve=DataInterface() {propertyValues {Curve {keys [{arriveTangent=6.5413, time=0, value=0}, <at 0.1 1>]}}}
      CurveIndex=$Particles.Age
      Scale Curve=&RandomRangeFloat(Minimum=0.8, Maximum=1)
     )
     B=&FloatFromCurve(FloatCurve=DataInterface() {propertyValues {Curve {keys [
          {arriveTangent=6.5413, interpMode=RCIM_Cubic, tangentMode=RCTM_Break, time=0, value=1}
          {arriveTangent=1.114, interpMode=RCIM_Cubic, tangentMode=RCTM_Break, time=1, value=1.35}
         ]}}})
    )) {isSetParameters=true}
   &DynamicMaterialParameters(
    Index 0 Param 2=$Particles.BlurMaskScale
    Index 0 Param 3=$Particles.MarkFlat
    Index 0 Param 4=&MakeFloatFromVector(Vector=$Particles.Scale, Channel=::NewEnumerator0)
    Index 1 Param 2=&MaskFloatBySpawnGroup(Masked Float=0, Passthrough Float=1)
    Index 1 Param 3=&RandomRangeFloat(Minimum=0, Maximum=5)
    Index 1 Param 4=0.75
    Index 2 Param 1=$Particles.TintScale
    Index 2 Param 2=1.5
    Index 3 Param 1=&MakeFloatFromVector4(
     Vector 4=&self:Vector4FromQuat(QUat:Quat4f=$Particles.MeshOrientation)
     Channel=::NewEnumerator0
    )
    Index 3 Param 2=&MakeFloatFromVector4(
     Vector 4=&self:Vector4FromQuat(QUat:Quat4f=$Particles.MeshOrientation)
     Channel=::NewEnumerator1
    )
    Index 3 Param 3=&MakeFloatFromVector4(
     Vector 4=&self:Vector4FromQuat(QUat:Quat4f=$Particles.MeshOrientation)
     Channel=::NewEnumerator2
    )
    Index 3 Param 4=&MakeFloatFromVector4(
     Vector 4=&self:Vector4FromQuat(QUat:Quat4f=$Particles.MeshOrientation)
     Channel=::NewEnumerator3
    )
    Module.Param0WriteEnabled=true
    Module.Param0WriteEnabled1=true
    Module.Param0WriteEnabled2=true
    Module.Param0WriteEnabled3=true
    Module.Param1WriteEnabled=true
    Module.Param1WriteEnabled1=true
    Module.Param1WriteEnabled2=true
    Module.Param1WriteEnabled3=true
    Module.Param2WriteEnabled=true
    Module.Param2WriteEnabled1=true
    Module.Param2WriteEnabled2=true
    Module.Param2WriteEnabled3=true
    Module.Param3WriteEnabled=true
    Module.Param3WriteEnabled1=true
    Module.Param3WriteEnabled2=true
    Module.Param3WriteEnabled3=true
    Write Parameter Index 1=true
    Write Parameter Index 2=true
    Write Parameter Index 3=true
   )
   &ScaleColor(
    Scale Alpha=&FloatFromCurve(FloatCurve=DataInterface() {propertyValues {Curve {keys [
         {tangentMode=RCTM_User, time=0, value=1}
         {arriveTangent=-0.0405, interpMode=RCIM_Cubic, leaveTangent=-0.0405, tangentMode=RCTM_User, time=0.4, value=1}
         {arriveTangent=-5.5746, interpMode=RCIM_Cubic, leaveTangent=-5.5746, tangentMode=RCTM_User, time=1, value=0}
        ]}}})
    ScaleA=true
    ScaleRGB=true
   )
  ]
  rendererClasses [&NiagaraMeshRendererProperties]
  renderers [
   {rendererClass=&NiagaraMeshRendererProperties, rendererIndex=0}
   {rendererClass=&NiagaraMeshRendererProperties, rendererIndex=1}
   {rendererClass=&NiagaraMeshRendererProperties, rendererIndex=2}
   {rendererClass=&NiagaraMeshRendererProperties, rendererIndex=3}
  ]
 }}
summary {
 systemName=NS_Mud_Splash_Landing
 emitters [{emitterName=Footprint, rendererClasses [&NiagaraMeshRendererProperties]}]
 userVariables [
  User.bIsRight {type=NiagaraBool}
  User.FootDirection {type=Vector3f, description="", defaultValue {struct=Vector3f, value=<0 0 0>}}
  User.FootDirectionArray {
   type=NiagaraDataInterfaceArrayFloat3
   description=""
   defaultValue {
    struct=NiagaraExt_VariableValue_DataInterface
    value {
     dataInterface=&self:NiagaraDataInterfaceArrayFloat3_0
     dataInterfaceClass=&NiagaraDataInterfaceArrayFloat3
    }
   }
  }
  User.FootPosition {type=NiagaraPosition, description="", defaultValue {}}
  User.FootPositionArray {
   type=NiagaraDataInterfaceArrayPosition
   description=""
   defaultValue {
    struct=NiagaraExt_VariableValue_DataInterface
    value {
     dataInterface=&self:NiagaraDataInterfaceArrayPosition_0
     dataInterfaceClass=&NiagaraDataInterfaceArrayPosition
    }
   }
  }
  User.GroundNormal {type=Vector3f, description="", defaultValue {struct=Vector3f, value=<0 0 0>}}
  User.SkeletalMesh {
   type=NiagaraDataInterfaceSkeletalMesh
   defaultValue {
    struct=NiagaraExt_VariableValue_DataInterface
    value {dataInterface=&self:NiagaraDataInterfaceSkeletalMesh_0}
   }
  }
 ]
}
