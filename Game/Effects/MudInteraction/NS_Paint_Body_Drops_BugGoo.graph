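# Niagara system that dresses a character's body in bug goo with dangling drips, derived from a mud/paint effect.
# Burst-spawned leg, upper-body and decal meshes follow indexed bones or sockets; Blueprint vectors drive decal materials, and MudRemoveTimer shrinks and fades body meshes.
# Pendulum-constrained mesh drips hang from sampled skeletal-mesh points, passing pivot and tip positions to their material.
# Both drop emitters (GPU, CPU collision-export) and the splash emitter are disabled.
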NiagaraSystem /Game/Effects/MudInteraction/NS_Paint_Body_Drops_BugGoo
dependencies {
 usedDataInterfaces [
  &NiagaraDataInterfaceCurve
  &NiagaraDataInterfaceArrayColor
  &NiagaraDataInterfaceSpriteRendererInfo
  /Script/Niagara.NiagaraDataInterfaceExport
  &NiagaraDataInterfaceColorCurve
 ]
 usedDynamicInputs [
  &Lerp_Vector
  &RandomRangeInt
  &SetBoolByIntComparison
 ]
 usedModules [
  &self:SystemSpawnScript.SystemScriptSource.SystemScriptGraph.NiagaraNodeAssignment_0.SetVariables_#1
  &self:SystemSpawnScript.SystemScriptSource.SystemScriptGraph.NiagaraNodeAssignment_2.SetVariables_#2
  &self:SystemSpawnScript.SystemScriptSource.SystemScriptGraph.NiagaraNodeAssignment_1.SetVariables_#3
  &self:SystemSpawnScript.SystemScriptSource.SystemScriptGraph.NiagaraNodeAssignment_5.SetVariables_#4
  &self:SystemSpawnScript.SystemScriptSource.SystemScriptGraph.NiagaraNodeAssignment_9.SetVariables_#5
  &self:SystemSpawnScript.SystemScriptSource.SystemScriptGraph.NiagaraNodeAssignment_6.SetVariables_#6
  &self:SystemSpawnScript.SystemScriptSource.SystemScriptGraph.NiagaraNodeAssignment_4.SetVariables_#7
  &SystemState
  &self:SetPeeloutFXArrayIndex
  &self:SystemSpawnScript.SystemScriptSource.SystemScriptGraph.NiagaraNodeAssignment_3.SetVariables_#8
 ]
}
emitters {
 Drops_CPU_with_event {
  bEnabled=false
  emitterUpdateScript [&EmitterState(Life Cycle Mode=::NewEnumerator1, Loop Duration=2, UseLoopDelay=false)]
  particleSpawnScript [&InitializeParticle(
    Lifetime Min=1.5
    Lifetime Mode=::NewEnumerator1
    Sprite Size Mode=::NewEnumerator1
    Uniform Sprite Size Max=4
    Uniform Sprite Size Min=3
    UsePositionOffset=false
    Write Material Random=true
   )]
  particleUpdateScript [
   &ParticleState()
   &SkeletalMeshLocation(
    Bone Sampling Mode=::NewEnumerator0
    MaskSpawnGroup=false
    Orientation Sampling=::NewEnumerator1
    OverrideRandomSeed=false
    Position Sampling=::NewEnumerator2
    Skeletal Mesh=$User.SkeletalMesh
   )
   &GravityForce()
   &Drag(Drag=&RandomRangeFloat(Minimum=1, Maximum=2), Use Linear Drag=true, Use Rotational Drag=true)
   &Collision(
    Age Colliding Particles=false
    CPU Trace Vector Length Multiplier=3
    Enable Rest State=false
    Max CPU Trace Length=100
    Module.EnableMaxCollisionCount=false
    Module.Randomize Collision Normal Vector=false
    Restitution=0
   )
   /Niagara/Modules/Events/GenerateCollisionEvent(Module.EventDelay=true, Module.Use Velocity Threshold=true)
   &SolveForcesAndVelocity(Clamp Velocity=false, Limit Acceleration=false)
   /SetVariables(
    Particles.MeshVec:Vector3f=&NormalizeVector(
     Vector To Normalize=&Subtract_Vector(
      A=&ConvertPositionToVector(Input Position=$Particles.OriginPos, Conversion Operation:ENiagara_LWCConvertPosToVec=::NewEnumerator0)
      B=&ConvertPositionToVector(Input Position=$Particles.Position, Conversion Operation:ENiagara_LWCConvertPosToVec=::NewEnumerator0)
     )
     Safe Normalize=true
     Safe Fallback Threshold=0
     Safe Fallback Vector=<0 0 0.001>
    )
    Particles.PosVec:Vector3f=&ConvertPositionToVector(Input Position=$Particles.Position, Conversion Operation:ENiagara_LWCConvertPosToVec=::NewEnumerator0)
   ) {isSetParameters=true, module=SetVariables_#9}
   /Niagara/Modules/Update/Orientation/V2/OrientMeshToVector(Look At Direction:Vector3f=$Particles.MeshVec, Override Quat=false, UseOptionalSideAxis=false, UseOptionalUpAxis=false)
   &DynamicMaterialParameters(
    Index 0 Param 2=5
    Index 0 Param 3=10
    Index 0 Param 4=5
    Index 1 Param 1=&FloatFromCurve(
     FloatCurve=DataInterface() {propertyValues {Curve {keys [
         {arriveTangent=-3.7492, interpMode=RCIM_Cubic, leaveTangent=-3.7492, tangentMode=RCTM_User, time=0, value=1}
         {arriveTangent=-0.7477, interpMode=RCIM_Cubic, leaveTangent=-0.7477, tangentMode=RCTM_User, time=0.3, value=0.3}
         {interpMode=RCIM_Cubic, leaveTangent=-1.9286, time=1, value=0}
        ]}}}
     Scale Curve=&RandomRangeFloat(Minimum=2, Maximum=2.25)
    )
    Module.Param0WriteEnabled=true
    Module.Param0WriteEnabled1=true
    Module.Param1WriteEnabled=true
    Module.Param1WriteEnabled1=false
    Module.Param2WriteEnabled=true
    Module.Param2WriteEnabled1=false
    Module.Param3WriteEnabled=true
    Module.Param3WriteEnabled1=false
    Write Parameter Index 1=true
   )
   &ScaleMeshSize(Scale Factor=&Multiply_VectorByFloat(
     Vector=<1 1 1>
     Float=&FloatFromCurve(FloatCurve=DataInterface() {propertyValues {Curve {keys [{interpMode=RCIM_Cubic, tangentMode=RCTM_User, time=0.1, value=0}, {tangentMode=RCTM_User, time=0.7, value=1}]}}})
    ))
   &ScaleColor(Scale RGB=<0 0 0>, ScaleA=false, ScaleRGB=true)
   /Niagara/Modules/ExportParticleData/ExportParticleDataToBlueprint(
    Condition To Export Data:NiagaraBool=$Output.Collision.CollisionValid
    Export Particle Data Interface=DataInterface() {propertyValues {
      GPUAllocationFixedSize=32
      GPUAllocationMode=FixedSize
      GPUAllocationPerParticleSize=1
      CallbackHandlerParameter {parameter {name=User.BP_Hit, typeDefHandle {registeredTypeIndex=1}, varData []}}
     }}
    "Float To Send (As Struct Size)":NiagaraFloat=&MakeFloatFromInt(Int=$Particles.Collision.CollisionQueryAndResponse.Collision1MaterialIndex)
    UseDelayTimer=false
   )
   &KillParticles(Kill Particles=$Output.Collision.CollisionValid)
  ]
  rendererClasses [&NiagaraSpriteRendererProperties]
  renderers [{rendererClass=&NiagaraSpriteRendererProperties}]
 }
 GPU_Drops {
  bEnabled=false
  simTarget=GPUComputeSim
  emitterUpdateScript [
   &EmitterState(Life Cycle Mode=::NewEnumerator1, Loop Duration=2, UseLoopDelay=false)
   &SpawnRate(SpawnRate=100, Use Spawn Probability=false)
   &SpawnRate(SpawnGroup=1, SpawnRate=5, Use Spawn Probability=false) {module=SpawnRate001}
   /SetVariables(Emitter.DropsColor:LinearColor=&SelectLinearColorFromArray(
     Array Sampling Mode=::NewEnumerator1
     Direct Array Index=$User.MudColor
     Color Selection Array=DataInterface() {propertyValues {ColorData [<rgba 0.015 0.0049 0.0024 1>, <rgba 0 0.6717 1 1>, <rgba 0.5672 0 1 1>, <rgba 1 0.7902 0 1>]}}
     Random Seed=?
    )) {isSetParameters=true}
  ]
  particleSpawnScript [
   &InitializeParticle(
    Color=<rgba 6 6 6 5>
    Lifetime=0
    Mass Max=1.1
    Mass Min=0.9
    Mass Mode=::NewEnumerator3
    Sprite Size Mode=::NewEnumerator1
    Sprite UV Mode=::NewEnumerator3
    Uniform Sprite Size Min=6
    UsePositionOffset=false
    Write Material Random=true
   )
   &ShapeLocation(MaskSpawnGroup=false, Override Seed=false, Shape Origin=<0 0 -1>, Sphere Radius=4, UseOwnerScale=true)
   &SkeletalMeshLocation(
    Bone Sampling Mode=::NewEnumerator0
    "Inherit Velocity (Scale)"=0.0005
    MaskSpawnGroup=false
    Orientation Sampling=::NewEnumerator1
    OverrideRandomSeed=false
    Sampled Position Offset=$Particles.Initial.ShapeLocation.ShapePosition
    Sampled Position Offset Coordinate Space:ENiagaraSkelMeshTransforms=::NewEnumerator0
    Skeletal Mesh=$User.Skeletal Mesh3
    Velocity Sampling=::NewEnumerator0
    VelocitySpeedLimit=false
   )
   &ApplyInitialForces()
   /SetVariables(
    Particles.Lifetime=&MaskFloatBySpawnGroup(
     Masked Float=&Multiply_Float(
      A=$Particles.Lifetime
      B=&MakeCustomFloatFromBool(
       Bool=&SetBoolByFloatComparison(A=HlslExpression() {hlslExpression="length(Particles.SkeletalMeshLocation.SampledVelocity)"}, B=500)
       True Float=1
       False Float=0
      )
     )
     Passthrough Float=1
    )
    Particles.Stored=1
   ) {isSetParameters=true}
   &SubUVAnimation(
    "Sprite Renderer "=DataInterface() {propertyValues {SpriteRenderer=&self:GPU_Drops_0.NiagaraSpriteRendererProperties_0}}
    SubUV Animation Mode:ENiagara_SubUVLookupModeV2=::NewEnumerator2
    UseEndFrame=false
    UseStartFrame=false
   )
  ]
  particleUpdateScript [
   &ParticleState()
   &GravityForce(Gravity=&Multiply_VectorByFloat(Vector=<0 0 -1500>, Float=1))
   &Drag(Drag=0.25, Use Linear Drag=false, Use Rotational Drag=false)
   &Collision(
    Advanced Aging Rate=0.1
    Age Colliding Particles=false
    Collision Enabled=&SetBoolByFloatComparison(
     A=1
     B=&FloatFromCurve(
      FloatCurve=DataInterface() {propertyValues {Curve {keys [<at 0.1 0>, {tangentMode=RCTM_User, time=0.26, value=1.1}]}}}
      CurveIndex=$Particles.Age
     )
     Comparison Type=::NewEnumerator6
    )
    Friction=10
    Friction During a Bounce=10000
    GPU Collision Type=::NewEnumerator2
    Maximum Penetration Correction Distance=5
    Method for Calculating Particle Radius:ENiagaraCalculateRadiusOptions=::NewEnumerator1
    Module.Control Rest State Via Surface Normal=false
    Module.EnableMaxCollisionCount=false
    Module.Randomize Collision Normal Vector=false
    Percentage of Penetration Before Rest=0
    Rest State Time Range=0.1
    Restitution=0
   )
   &ScaleColor(
    Scale Alpha=&FloatFromCurve(FloatCurve=DataInterface() {propertyValues {Curve {keys [
         <at 0.9 1.0113 RCIM_Cubic>
         {arriveTangent=-5.7525, interpMode=RCIM_Cubic, leaveTangent=-5.7525, tangentMode=RCTM_User, time=1.0011, value=0.002}
        ]}}})
    ScaleA=true
    ScaleRGB=true
   )
   &SolveForcesAndVelocity(Clamp Velocity=false, Limit Acceleration=false)
   &self:ScratchModule(Velo:Vector3f=$Particles.Velocity) {module=ScratchModule}
   /SetVariables(Particles.NewNAge=HlslExpression() {hlslExpression="(Particles.Age - Particles.Stored) * (1.0/ (1.0-Particles.Stored+0))"}) {isSetParameters=true}
   &ScaleSpriteSize(
    Scale Sprite Size Mode=::NewEnumerator0
    Uniform Scale Factor=&MakeCustomFloatFromBool(
     Bool=&SetBoolByFloatComparison(A=0.1, B=HlslExpression() {hlslExpression="length(Particles.Velocity)"})
     True Float=&FloatFromCurve(
      FloatCurve=DataInterface() {propertyValues {Curve {keys [{tangentMode=RCTM_User, time=0, value=0.1}, {tangentMode=RCTM_User, time=0.5, value=1}]}}}
      CurveIndex=$Particles.NewNAge
      Scale Curve=7
     )
     False Float=&FloatFromCurve(FloatCurve=DataInterface() {propertyValues {Curve {keys [{tangentMode=RCTM_User, time=0, value=0}, {tangentMode=RCTM_User, time=0.1, value=1}]}}})
    )
   )
   /SetVariables(
    Particles.CameraOffset:NiagaraFloat=&MakeCustomFloatFromBool(
     Bool=&SetBoolByFloatComparison(A=1, B=HlslExpression() {hlslExpression="length(Particles.Velocity)"})
     True Float=5
     False Float=1
    )
    Particles.Lifetime=&MakeCustomFloatFromBool(
     Bool=&SetBoolByFloatComparison(A=1, B=HlslExpression() {hlslExpression="length(Particles.Velocity)"})
     True Float=&Multiply_Float(A=1, B=&Add_Float(A=$Particles.Stored, B=0.25))
     False Float=$Particles.Lifetime
    )
   ) {isSetParameters=true}
   &ScaleSpriteSizeBySpeed(Max Scale Factor=<0.5 2>, Min Scale Factor=<0.5 0.6>, Velocity Threshold=500)
  ]
  rendererClasses [&NiagaraSpriteRendererProperties]
  renderers [{rendererClass=&NiagaraSpriteRendererProperties}]
 }
 Mud_Body_Decal {
  emitterUpdateScript [
   &EmitterState()
   &SpawnBurst_Instantaneous(Spawn Count=7, Use Spawn Probability=false, UseLoopCountLimit=true)
  ]
  particleSpawnScript [
   &PartitionParticles(Number Of Partitions=7)
   &InitializeParticle(
    Color Mode=::NewEnumerator0
    Lifetime=1000000000
    Mesh Scale Mode=::NewEnumerator3
    Mesh Uniform Scale=&SelectFloatFromArray(
     Direct Array Index=&Add_Integer(A=0, B=$Particles.PartitionParticles.Partition)
     Float Selection Array=$User.SpriteSize Array
    )
    UsePositionOffset=false
    Write Material Random=true
   )
   &CameraOffset(Camera Offset Amount=20)
   /SetVariables() {isSetParameters=true}
   &DynamicMaterialParameters(
    Index 0 Param 1=&Add_Float(A=&MakeFloatFromVector4(Vector 4=$User.InputFromBP1, Channel=::NewEnumerator0), B=0)
    Index 0 Param 2=&MakeFloatFromVector4(Vector 4=$User.InputFromBP1, Channel=::NewEnumerator1)
    Index 0 Param 3=&Add_Float(A=&MakeFloatFromVector4(Vector 4=$User.InputFromBP1, Channel=::NewEnumerator2), B=0)
    Index 0 Param 4=&MakeFloatFromVector4(Vector 4=$User.InputFromBP1, Channel=::NewEnumerator3)
    Index 1 Param 1=&MakeFloatFromVector4(Vector 4=$User.InputFromBP2, Channel=::NewEnumerator0)
    Index 1 Param 2=&MakeFloatFromVector4(Vector 4=$User.InputFromBP2, Channel=::NewEnumerator1)
    Index 1 Param 3=&MakeFloatFromVector4(Vector 4=$User.InputFromBP2, Channel=::NewEnumerator2)
    Module.Param0WriteEnabled=true
    Module.Param0WriteEnabled1=true
    Module.Param1WriteEnabled=true
    Module.Param1WriteEnabled1=true
    Module.Param2WriteEnabled=true
    Module.Param2WriteEnabled1=true
    Module.Param3WriteEnabled=true
    Module.Param3WriteEnabled1=false
    Write Parameter Index 1=true
   ) {module=DynamicMaterialParameters001}
  ]
  particleUpdateScript [
   &ParticleState()
   &SkeletalMeshLocation(
    Bone / Socket Index=&Add_Integer(A=17, B=$Particles.Initial.PartitionParticles.Partition)
    Bone Sampling Mode=::NewEnumerator4
    MaskSpawnGroup=false
    Orientation Sampling=::NewEnumerator1
    Sampled Position Offset=&SelectVectorFromArray(
     Direct Array Index=&Add_Integer(A=0, B=$Particles.Initial.PartitionParticles.Partition)
     Vector Selection Array=DataInterface() {propertyValues {FloatData [<0 0 0>, <-25 0 0>, <-20 0 0>, <-2 0 0>, <25 0 0>, <20 0 0>, <2 -5 0>]}}
    )
    Skeletal Mesh=$User.SkeletalMesh
   )
   &DynamicMaterialParameters(
    Index 0 Param 1=&MakeFloatFromVector4(Vector 4=$User.InputFromBP1, Channel=::NewEnumerator0)
    Index 0 Param 2=&MakeFloatFromVector4(Vector 4=$User.InputFromBP1, Channel=::NewEnumerator1)
    Index 0 Param 3=&MakeFloatFromVector4(Vector 4=$User.InputFromBP1, Channel=::NewEnumerator2)
    Index 0 Param 4=&MakeFloatFromVector4(Vector 4=$User.InputFromBP1, Channel=::NewEnumerator3)
    Index 1 Param 1=&MakeFloatFromVector4(Vector 4=$User.InputFromBP2, Channel=::NewEnumerator0)
    Index 1 Param 2=&MakeFloatFromVector4(Vector 4=$User.InputFromBP2, Channel=::NewEnumerator1)
    Index 1 Param 3=&MakeFloatFromVector4(Vector 4=$User.InputFromBP2, Channel=::NewEnumerator2)
    Module.Param0WriteEnabled=true
    Module.Param0WriteEnabled1=true
    Module.Param1WriteEnabled=false
    Module.Param1WriteEnabled1=true
    Module.Param2WriteEnabled=true
    Module.Param2WriteEnabled1=true
    Module.Param3WriteEnabled=false
    Module.Param3WriteEnabled1=false
    Write Parameter Index 1=true
   )
   /SetVariables(Particles.Counter:NiagaraFloat=&Add_Float(A=$Particles.Counter, B=$Engine.DeltaTime)) {isSetParameters=true}
  ]
  rendererClasses [&NiagaraMeshRendererProperties]
  renderers [{rendererClass=&NiagaraMeshRendererProperties, rendererIndex=0}]
 }
 NE_Parent_Mud_Splash {
  bEnabled=false
  emitterUpdateScript [&EmitterState(Enable Distance Culling=true, MaxDistance=9000, Module.bUseMaxDistance=true, Module.bUseMinDistance=false, Scalability Mode=::NewEnumerator1)]
  particleSpawnScript [
   &InitializeParticle(
    Color=<rgba 0.5312 0.5312 0.5312 1>
    Lifetime Max=0.75
    Lifetime Min=0.5
    Lifetime Mode=::NewEnumerator1
    Position Offset=&Multiply_VectorByFloat(
     Vector=<0 0 70>
     Float=&FloatFromCurve(
      FloatCurve=DataInterface() {propertyValues {Curve {keys [{tangentMode=RCTM_User, time=0, value=0.1}, {tangentMode=RCTM_User, time=1, value=1}]}}}
      CurveIndex=$System.NormalizedLoopAge
     )
    )
    Sprite Rotation Mode=::NewEnumerator2
    Sprite Size Mode=::NewEnumerator1
    Sprite UV Mode=::NewEnumerator3
    Uniform Sprite Size Max=100
    Uniform Sprite Size Min=90
    UsePositionOffset=true
    Write Material Random=true
   )
   &SubUVAnimation(
    "Sprite Renderer "=DataInterface() {propertyValues {SpriteRenderer=&self:NE_Parent_Mud_Splash.NiagaraSpriteRendererProperties_0}}
    SubUV Animation Mode:ENiagara_SubUVLookupModeV2=::NewEnumerator2
    UseEndFrame=false
    UseStartFrame=false
   )
   &ShapeLocation(
    Apply To Particle Position=::NewEnumerator1
    MaskSpawnGroup=false
    Offset Coordinate Space=::World
    Offset Mode=::NewEnumerator0
    Override Seed=false
    Rotation Coordinate Space=::World
    Rotation Mode=::NewEnumerator1
    Sphere Radius=1
    UseOwnerScale=true
   ) {module=ShapeLocation001}
   &AddVelocity(Velocity=$Particles.ShapeLocation001.ShapeNormal, Velocity Speed Scale=70)
   &AddVelocity(Rotation Coordinate Space=::World, Velocity=<0 0 1>, Velocity Speed Scale=150) {module=AddVelocity001}
   &CameraOffset(Camera Offset Amount=15)
  ]
  particleUpdateScript [
   &ParticleState()
   &GravityForce(Gravity=<0 0 -1200>)
   &InheritSourceMovement(
    Applied Position Scale=&VectorFromFloat(Value=0.95)
    ApplyForce=true
    ApplyForceFalloff=false
    ApplyPosition=true
    ApplyPositionFalloff=false
    Module.LimitForceSpeed=false
   )
   &SolveForcesAndVelocity(Clamp Velocity=false, Limit Acceleration=false)
   &ScaleSpriteSize(Uniform Curve Sprite Scale=DataInterface() {propertyValues {Curve {keys [
        {arriveTangent=4.1785, interpMode=RCIM_Cubic, leaveTangent=4.1785, tangentMode=RCTM_User, time=0, value=0.15}
        {arriveTangent=1.1179, interpMode=RCIM_Cubic, leaveTangent=1.1179, tangentMode=RCTM_User, time=0.2042, value=0.6228}
        {arriveTangent=0.0224, interpMode=RCIM_Cubic, leaveTangent=0.0224, tangentMode=RCTM_User, time=1, value=1}
       ]}}})
   &ScaleColor(
    Scale Alpha=&FloatFromCurve(FloatCurve=DataInterface() {propertyValues {Curve {keys [<at 0 1 RCIM_Cubic>, {arriveTangent=-2.0383, interpMode=RCIM_Cubic, leaveTangent=-2.0383, tangentMode=RCTM_User, time=1, value=0}]}}})
    Scale RGB=<0 0 0>
    ScaleA=true
    ScaleRGB=true
   )
  ]
  rendererClasses [&NiagaraSpriteRendererProperties]
  renderers [{rendererClass=&NiagaraSpriteRendererProperties}]
 }
 Pendulum002 {
  emitterUpdateScript [
   &EmitterState(
    Inactive Response=::NewEnumerator1
    Life Cycle Mode=::NewEnumerator1
    Loop Behavior=::NewEnumerator1
    Loop Duration=100000000
    UseLoopDelay=false
   )
   &SpawnRate(SpawnRate=1.5, Use Spawn Probability=false)
  ]
  particleSpawnScript [
   &InitializeParticle(
    Color=&ColorFromCurve(ColorCurve=DataInterface() {propertyValues {
       AlphaCurve {keys [<at 0 1 RCIM_Cubic>,<at 1 0 RCIM_Cubic>]}
       BlueCurve {keys [<at 0 1 RCIM_Cubic>]}
       GreenCurve {keys [<at 0 1 RCIM_Cubic>]}
       RedCurve {keys [<at 0 1 RCIM_Cubic>]}
      }})
    Lifetime Max=2.5
    Lifetime Min=2
    Lifetime Mode=::NewEnumerator1
    Mesh Scale=&Multiply_VectorByFloat(Vector=<0.6 0.6 0.7>, Float=&RandomRangeFloat(Minimum=0.4, Maximum=0.8))
    Mesh Scale Mode=::NewEnumerator4
    Sprite Rotation Mode=::NewEnumerator2
    Sprite Size Mode=::NewEnumerator3
    Sprite UV Mode=::NewEnumerator3
    Uniform Sprite Size=&RandomRangeFloat(Minimum=5, Maximum=10)
    UsePositionOffset=false
    Write Material Random=true
   )
   &CameraOffset(Camera Offset Amount=1)
  ]
  particleUpdateScript [
   &ParticleState()
   &SkeletalMeshLocation(
    Bone Sampling Mode=::NewEnumerator0
    MaskSpawnGroup=false
    Orientation Sampling=::NewEnumerator1
    OverrideRandomSeed=false
    Position Sampling=::NewEnumerator2
    Skeletal Mesh=$User.Skeletal Mesh3
   )
   /SetVariables(Particles.OriginPos:NiagaraPosition=&AddVectorToPosition(
     Position=$Particles.SkeletalMeshLocation.SampledPosition
     Vector=&RandomRangeVector(Minimum=<-1 -1 -1>, Maximum=<1 1 1>)
    )) {isSetParameters=true}
   &GravityForce(Gravity=<0 0 -2000>)
   &Drag(Drag=&RandomRangeFloat(Minimum=1, Maximum=2), Use Linear Drag=true, Use Rotational Drag=true)
   &PendulumSetup(
    Module.Enable Spring Driver=false
    Pendulum Length:NiagaraFloat=&RandomRangeFloat(Minimum=20, Maximum=40)
    Pendulum Pivot:Vector3f=&ConvertPositionToVector(Input Position=$Particles.OriginPos, Conversion Operation:ENiagara_LWCConvertPosToVec=::NewEnumerator0)
   )
   &SolveForcesAndVelocity(Clamp Velocity=false, Limit Acceleration=false)
   &GenerateDeathEvent(Use Event Probability=true, "Vector to Send as Vector 1 (Position)":NiagaraPosition=$Particles.OriginPos)
   &PendulumConstraint(Calculate and Output Potential Energy=true)
   &CalculateAccurateVelocity(Calculate Accurate Velocity On First Frame=true)
   /SetVariables(
    Particles.MeshVec:Vector3f=&NormalizeVector(
     Vector To Normalize=&Subtract_Vector(
      A=&ConvertPositionToVector(Input Position=$Particles.OriginPos, Conversion Operation:ENiagara_LWCConvertPosToVec=::NewEnumerator0)
      B=&ConvertPositionToVector(Input Position=$Particles.Position, Conversion Operation:ENiagara_LWCConvertPosToVec=::NewEnumerator0)
     )
     Safe Normalize=true
     Safe Fallback Threshold=0
     Safe Fallback Vector=<0 0 0.001>
    )
    Particles.PosVec:Vector3f=&ConvertPositionToVector(Input Position=$Particles.Position, Conversion Operation:ENiagara_LWCConvertPosToVec=::NewEnumerator0)
   ) {isSetParameters=true, module=SetVariables_#9}
   &OrientMeshToVector(Look At Direction:Vector3f=$Particles.MeshVec, Override Quat=false, UseOptionalSideAxis=false, UseOptionalUpAxis=false)
   &DynamicMaterialParameters(
    Index 0 Param 4=&FloatFromCurve(FloatCurve=DataInterface() {propertyValues {Curve {keys [<at 0 1>,<at 1 0>]}}}, Scale Curve=0.5)
    Index 2 Param 1=&MakeFloatFromVector(
     Vector=&ConvertPositionToVector(Input Position=$Particles.OriginPos, Conversion Operation:ENiagara_LWCConvertPosToVec=::NewEnumerator0)
     Channel=::NewEnumerator0
    )
    Index 2 Param 2=&MakeFloatFromVector(
     Vector=&ConvertPositionToVector(Input Position=$Particles.OriginPos, Conversion Operation:ENiagara_LWCConvertPosToVec=::NewEnumerator0)
     Channel=::NewEnumerator1
    )
    Index 2 Param 3=&MakeFloatFromVector(
     Vector=&ConvertPositionToVector(Input Position=$Particles.OriginPos, Conversion Operation:ENiagara_LWCConvertPosToVec=::NewEnumerator0)
     Channel=::NewEnumerator2
    )
    Index 2 Param 4=1
    Index 3 Param 1=&MakeFloatFromVector(
     Vector=&ConvertPositionToVector(Input Position=$Particles.Position, Conversion Operation:ENiagara_LWCConvertPosToVec=::NewEnumerator0)
     Channel=::NewEnumerator0
    )
    Index 3 Param 2=&MakeFloatFromVector(
     Vector=&ConvertPositionToVector(Input Position=$Particles.Position, Conversion Operation:ENiagara_LWCConvertPosToVec=::NewEnumerator0)
     Channel=::NewEnumerator1
    )
    Index 3 Param 3=&MakeFloatFromVector(
     Vector=&ConvertPositionToVector(Input Position=$Particles.Position, Conversion Operation:ENiagara_LWCConvertPosToVec=::NewEnumerator0)
     Channel=::NewEnumerator2
    )
    Index 3 Param 4=0.2
    Module.Param0WriteEnabled=false
    Module.Param0WriteEnabled2=true
    Module.Param0WriteEnabled3=true
    Module.Param1WriteEnabled=false
    Module.Param1WriteEnabled2=true
    Module.Param1WriteEnabled3=true
    Module.Param2WriteEnabled=false
    Module.Param2WriteEnabled2=true
    Module.Param2WriteEnabled3=true
    Module.Param3WriteEnabled=true
    Module.Param3WriteEnabled2=true
    Module.Param3WriteEnabled3=true
    Write Parameter Index 2=true
    Write Parameter Index 3=true
   )
   &ScaleMeshSize(Scale Factor=&Multiply_VectorByFloat(
     Vector=<1 1 1>
     Float=&FloatFromCurve(FloatCurve=DataInterface() {propertyValues {Curve {keys [
          {interpMode=RCIM_Cubic, tangentMode=RCTM_User, time=0, value=0}
          {tangentMode=RCTM_User, time=0.5, value=1}
          {arriveTangent=-1.8, interpMode=RCIM_Cubic, leaveTangent=-1.8, time=0.88, value=1}
          {arriveTangent=-3.0723, interpMode=RCIM_Cubic, leaveTangent=-3.0723, tangentMode=RCTM_User, time=1, value=0.1}
         ]}}})
    ))
  ]
  rendererClasses [&NiagaraMeshRendererProperties]
  renderers [{rendererClass=&NiagaraMeshRendererProperties, rendererIndex=0}]
 }
 Skeletal_Mesh_Particle_Legs {
  emitterUpdateScript [
   &EmitterState()
   &SpawnBurst_Instantaneous(Spawn Count=6, Use Spawn Probability=false, UseLoopCountLimit=true)
  ]
  particleSpawnScript [
   &PartitionParticles(Number Of Partitions=6)
   &InitialMeshOrientation(
    Mesh Axis To Orient=<0 0 1>
    Mesh Orientation Mode=::NewEnumerator2
    Orientation Vector=&SelectVectorFromArray(
     Direct Array Index=$Particles.PartitionParticles.Partition
     Vector Selection Array=DataInterface() {propertyValues {FloatData [<1 0 0>, <1 0 0>, <0 -1 0>, <-1 0 0>, <-1 0 0>, <0 1 0>]}}
    )
    Rotation=&SelectVectorFromArray(
     Direct Array Index=$Particles.PartitionParticles.Partition
     Vector Selection Array=DataInterface() {propertyValues {FloatData [<0 0 0>, <-0.02 0 0.4>, <0 0 0.5>, <0 0 0.8>, <0.06 0.3 1>, <-0.1 0 0.5>]}}
    )
    Use Rotation Vector=true
   )
   &InitializeParticle(
    Lifetime=1000000000
    Mesh Scale=&Multiply_Vector(
     A=<0.11 0.11 0.15>
     B=&SelectVectorFromArray(
      Direct Array Index=$Particles.PartitionParticles.Partition
      Vector Selection Array=DataInterface() {propertyValues {FloatData [<0.92 0.9 0.8>, <0.7 0.8 0.8>, <0.72 0.6 0.5>, <1.3 0.85 0.75>, <0.8 0.8 0.7>, <0.8 0.8 0.58>]}}
     )
    )
    Mesh Scale Mode=::NewEnumerator4
    Sprite Rotation Mode=::NewEnumerator2
    Sprite Size Mode=::NewEnumerator1
    Uniform Sprite Size Max=6
    Uniform Sprite Size Min=2
    UsePositionOffset=false
    Write Material Random=true
   )
   &ShapeLocation(
    Apply To Particle Position=::NewEnumerator1
    Cylinder Height=1
    Cylinder Height Midpoint:NiagaraFloat=&SelectFloatFromArray(Direct Array Index=$Particles.PartitionParticles.Partition, Float Selection Array=DataInterface() {propertyValues {FloatData [0.2,0]}})
    Cylinder Radius=1
    MaskSpawnGroup=false
    Non Uniform Scale=&SelectVectorFromArray(Direct Array Index=$Particles.PartitionParticles.Partition, Vector Selection Array=$System.Shape Scale Array)
    Offset Mode=::NewEnumerator0
    Override Seed=false
    Rotation Coordinate Space=::World
    Rotation Mode=::NewEnumerator1
    Shape Origin=&VectorFromFloat(Value=0)
    Shape Primitive=::NewEnumerator1
    Surface Only=true
    Use Endcaps In Surface Only Mode=false
    UseOwnerScale=true
   )
   &ShapeLocation(
    Apply To Particle Position=::NewEnumerator1
    MaskSpawnGroup=false
    Offset=<5 0 0>
    Offset Coordinate Space=::World
    Offset Mode=::NewEnumerator0
    Override Seed=false
    Rotation Coordinate Space=::World
    Rotation Mode=::NewEnumerator1
    Shape Origin=&VectorFromFloat(Value=0)
    Sphere Radius=1
    Sphere Surface Distribution=1
    UseOwnerScale=true
   ) {module=ShapeLocation001}
   &CameraOffset(Camera Offset Amount=1) {enabled=false}
   /SetVariables(Particles.Scale=&Multiply_Vector(A=$Particles.Scale, B=$System.Adjust Scale by Character Size)) {isSetParameters=true}
  ]
  particleUpdateScript [
   &ParticleState()
   &SkeletalMeshLocation(
    Bone / Socket Index=&Add_Integer(A=11, B=$Particles.PartitionParticles.Partition)
    Bone Sampling Mode=::NewEnumerator4
    MaskSpawnGroup=false
    Sampled Position Offset=&SelectVectorFromArray(
     Direct Array Index=$Particles.PartitionParticles.Partition
     Vector Selection Array=DataInterface() {propertyValues {FloatData [<0 0 1>, <0 1 0>, <-0.5 4.5 0>, <0 0 0>, <0 0 0>, <0 -5 0>]}}
    )
    Skeletal Mesh=$User.SkeletalMesh
    Use Sampled Mesh Attribute as Ribbon ID=false
   ) {module=SkeletalMeshLocation001}
   &DynamicMaterialParameters(
    Index 0 Param 4=&Multiply_Float(A=$Emitter.Age, B=0.03)
    Index 1 Param 4=$Particles.MaterialRandom
    Module.Param0WriteEnabled=false
    Module.Param0WriteEnabled1=false
    Module.Param1WriteEnabled=false
    Module.Param1WriteEnabled1=false
    Module.Param2WriteEnabled=false
    Module.Param2WriteEnabled1=false
    Module.Param3WriteEnabled=true
    Module.Param3WriteEnabled1=true
    Write Parameter Index 1=true
   )
   &SpriteRotationRate(Rotation Rate=&MakeCustomFloatFromBool(Bool=&RandomBool(), True Float=-1080, False Float=1080))
   &SolveForcesAndVelocity(Clamp Velocity=false, Limit Acceleration=false)
   /SetVariables(Particles.OriginPos:NiagaraPosition=$Particles.SkeletalMeshLocation001.SampledPosition) {isSetParameters=true, module=SetVariables_#10}
   &ScaleMeshSize(Scale Factor=&MakeVector(
     X=&MakeFloatFromVector(
      Vector=&VectorFromFloat(Value=&FloatFromCurve(FloatCurve=DataInterface() {propertyValues {Curve {keys [<at 0.1 1>,<at 1 0>]}}}, CurveIndex=$System.MudRemoveTimer))
      Channel=::NewEnumerator0
     )
     Y=&MakeFloatFromVector(
      Vector=&VectorFromFloat(Value=&FloatFromCurve(FloatCurve=DataInterface() {propertyValues {Curve {keys [<at 0.1 1>,<at 1 0>]}}}, CurveIndex=$System.MudRemoveTimer))
      Channel=::NewEnumerator0
     )
     Z=1
    ))
   &ScaleColor(
    Scale Alpha=&FloatFromCurve(FloatCurve=DataInterface() {propertyValues {Curve {keys [<at 0 1>,<at 1 -1>]}}}, CurveIndex=$System.MudRemoveTimer)
    ScaleA=true
    ScaleRGB=true
   )
  ]
  rendererClasses [&NiagaraMeshRendererProperties]
  renderers [{rendererClass=&NiagaraMeshRendererProperties, rendererIndex=0}]
 }
 Skeletal_Mesh_Particle_Upper_Body {
  emitterUpdateScript [
   &EmitterState()
   &SpawnBurst_Instantaneous(Spawn Count=8, Use Spawn Probability=false, UseLoopCountLimit=true)
  ]
  particleSpawnScript [
   &PartitionParticles(Number Of Partitions=8)
   &InitialMeshOrientation(
    Mesh Axis To Orient=<0 0 1>
    Mesh Orientation Mode=::NewEnumerator2
    Orientation Vector=&SelectVectorFromArray(
     Direct Array Index=$Particles.PartitionParticles.Partition
     Vector Selection Array=DataInterface() {propertyValues {FloatData [<-1 0 0>, <-1 0 0>, <-1 0 0.4>, <1 0 0>, <1 0 0>, <1 0 0>, <0 -0.05 -0.3>]}}
    )
    Rotation=&SelectVectorFromArray(
     Direct Array Index=$Particles.PartitionParticles.Partition
     Vector Selection Array=DataInterface() {propertyValues {FloatData [<0 0 0.4>, <0 0 0>, <0 0 0>, <0 -0.02 0.5>, <0 0 0>, <0 0 0.5>, <0 0 0.5>]}}
    )
    Use Rotation Vector=true
   )
   &InitializeParticle(
    Lifetime=1000000000
    Mesh Scale=&Multiply_Vector(
     A=<0.06 0.07 0.09>
     B=&SelectVectorFromArray(
      Direct Array Index=$Particles.PartitionParticles.Partition
      Vector Selection Array=DataInterface() {propertyValues {FloatData [<1.5 1.2 1.02>, <1.1 1.1 0.6>, <1 1 0.5>, <1 1.2 1>, <1.1 0.8 0.6>, <0.65 0.7 0.5>, <1.2 1.4 0.6>, <1.2 1.2 0.8>]}}
     )
    )
    Mesh Scale Mode=::NewEnumerator4
    Sprite Rotation Mode=::NewEnumerator2
    Sprite Size Mode=::NewEnumerator1
    Uniform Sprite Size Max=6
    Uniform Sprite Size Min=2
    UsePositionOffset=false
    Write Material Random=true
   )
   &ShapeLocation(
    Apply To Particle Position=::NewEnumerator1
    Cylinder Height=1
    Cylinder Height Midpoint:NiagaraFloat=&SelectFloatFromArray(Direct Array Index=$Particles.PartitionParticles.Partition, Float Selection Array=DataInterface() {propertyValues {FloatData [0.2,0]}})
    Cylinder Radius=1
    MaskSpawnGroup=false
    Non Uniform Scale=&SelectVectorFromArray(Direct Array Index=$Particles.PartitionParticles.Partition, Vector Selection Array=$System.Shape Scale Array)
    Offset Mode=::NewEnumerator0
    Override Seed=false
    Rotation Coordinate Space=::World
    Rotation Mode=::NewEnumerator1
    Shape Origin=&VectorFromFloat(Value=0)
    Shape Primitive=::NewEnumerator1
    Surface Only=true
    Use Endcaps In Surface Only Mode=false
    UseOwnerScale=true
   )
   &ShapeLocation(
    Apply To Particle Position=::NewEnumerator1
    MaskSpawnGroup=false
    Offset=<5 0 0>
    Offset Coordinate Space=::World
    Offset Mode=::NewEnumerator0
    Override Seed=false
    Rotation Coordinate Space=::World
    Rotation Mode=::NewEnumerator1
    Shape Origin=&VectorFromFloat(Value=0)
    Sphere Radius=1
    Sphere Surface Distribution=1
    UseOwnerScale=true
   ) {module=ShapeLocation001}
   &CameraOffset(Camera Offset Amount=&Multiply_Float(A=1, B=&MakeFloatFromVector(Vector=$System.Adjust Scale by Character Size, Channel=::NewEnumerator0)))
   /SetVariables(Particles.Scale=&Multiply_Vector(
     A=$Particles.Scale
     B=&Multiply_Vector(A=$System.Adjust Scale by Character Size, B=$System.Adjust Scale by Character Size)
    )) {isSetParameters=true}
  ]
  particleUpdateScript [
   &ParticleState()
   &SkeletalMeshLocation(
    Bone / Socket Index=&Add_Integer(A=2, B=$Particles.PartitionParticles.Partition)
    Bone Sampling Mode=::NewEnumerator4
    MaskSpawnGroup=false
    Sampled Position Offset=&SelectVectorFromArray(
     Direct Array Index=$Particles.PartitionParticles.Partition
     Vector Selection Array=DataInterface() {propertyValues {FloatData [<4 0 -1.6>, <1 0 1>, <0 3 0>, <-3 -0.6 1.4>, <-3 0.7 -1.9>, <0.2 -0.8 -0.1>, <2 -1 5>, <5 -3 16>]}}
    )
    Skeletal Mesh=$User.SkeletalMesh
    Use Sampled Mesh Attribute as Ribbon ID=false
   ) {module=SkeletalMeshLocation001}
   &DynamicMaterialParameters(
    Index 0 Param 4=&Multiply_Float(A=$Emitter.Age, B=0.03)
    Index 1 Param 4=$Particles.MaterialRandom
    Module.Param0WriteEnabled=false
    Module.Param0WriteEnabled1=false
    Module.Param1WriteEnabled=false
    Module.Param1WriteEnabled1=false
    Module.Param2WriteEnabled=false
    Module.Param2WriteEnabled1=false
    Module.Param3WriteEnabled=true
    Module.Param3WriteEnabled1=true
    Write Parameter Index 1=true
   )
   &SpriteRotationRate(Rotation Rate=&MakeCustomFloatFromBool(Bool=&RandomBool(), True Float=-1080, False Float=1080))
   &SolveForcesAndVelocity(Clamp Velocity=false, Limit Acceleration=false)
   /SetVariables(Particles.OriginPos:NiagaraPosition=$Particles.SkeletalMeshLocation001.SampledPosition) {isSetParameters=true, module=SetVariables_#10}
   &ScaleMeshSize(Scale Factor=&MakeVector(
     X=&MakeFloatFromVector(
      Vector=&VectorFromFloat(Value=&FloatFromCurve(FloatCurve=DataInterface() {propertyValues {Curve {keys [<at 0.1 1>,<at 1 0>]}}}, CurveIndex=$System.MudRemoveTimer))
      Channel=::NewEnumerator0
     )
     Y=&MakeFloatFromVector(
      Vector=&VectorFromFloat(Value=&FloatFromCurve(FloatCurve=DataInterface() {propertyValues {Curve {keys [<at 0.1 1>,<at 1 0>]}}}, CurveIndex=$System.MudRemoveTimer))
      Channel=::NewEnumerator0
     )
     Z=1
    ))
   &ScaleColor(
    Scale Alpha=&FloatFromCurve(FloatCurve=DataInterface() {propertyValues {Curve {keys [<at 0 1>,<at 1 -1>]}}}, CurveIndex=$System.MudRemoveTimer)
    ScaleA=true
    ScaleRGB=true
   )
  ]
  rendererClasses [&NiagaraMeshRendererProperties]
  renderers [{rendererClass=&NiagaraMeshRendererProperties, rendererIndex=0}]
 }
}
summary {
 systemName=NS_Paint_Body_Drops_BugGoo
 emitters [
  {emitterName=Skeletal_Mesh_Particle_Legs, rendererClasses [&NiagaraMeshRendererProperties]}
  {emitterName=Skeletal_Mesh_Particle_Upper_Body, rendererClasses [&NiagaraMeshRendererProperties]}
  {bEnabled=false, emitterName=GPU_Drops, simTarget=GPUComputeSim, rendererClasses [&NiagaraSpriteRendererProperties]}
  {bEnabled=false, emitterName=Drops_CPU_with_event, rendererClasses [&NiagaraSpriteRendererProperties]}
  {emitterName=Pendulum002, rendererClasses [&NiagaraMeshRendererProperties]}
  {bEnabled=false, emitterName=NE_Parent_Mud_Splash, rendererClasses [&NiagaraSpriteRendererProperties]}
  {emitterName=Mud_Body_Decal, rendererClasses [&NiagaraMeshRendererProperties]}
 ]
 userVariables [
  User.CharacterSizeID {type=NiagaraInt32}
  User.GooColorScheme {type=NiagaraInt32}
  User.Initial Rotation {type=NiagaraFloat, defaultValue=0, description=""}
  User.Initial Rotation Range {type=NiagaraFloat, defaultValue=0, description=""}
  User.InputFromBP1 {type=Vector4f, description="", defaultValue {struct=Vector4f, value=<0 0.2 10000 2>}}
  User.InputFromBP2 {type=Vector4f, description="", defaultValue {struct=Vector4f, value=<0 0 0 0>}}
  User.Mesh Scale Array {
   type=NiagaraDataInterfaceArrayFloat3
   description=""
   defaultValue {
    struct=NiagaraExt_VariableValue_DataInterface
    value {
     dataInterface=&self:NiagaraDataInterfaceArrayFloat3_0
     dataInterfaceClass=&NiagaraDataInterfaceArrayFloat3
    }
   }
  }
  User.MudColor {type=NiagaraInt32, defaultValue=2}
  User.MudRemove {type=NiagaraInt32, defaultValue=1}
  User.MudRemoveTimer {type=NiagaraFloat, defaultValue=0, description=""}
  User.Skeletal Mesh3 {
   type=NiagaraDataInterfaceSkeletalMesh
   defaultValue {
    struct=NiagaraExt_VariableValue_DataInterface
    value {dataInterface=&self:NiagaraDataInterfaceSkeletalMesh_1}
   }
  }
  User.SkeletalMesh {
   type=NiagaraDataInterfaceSkeletalMesh
   defaultValue {
    struct=NiagaraExt_VariableValue_DataInterface
    value {dataInterface=&self:NiagaraDataInterfaceSkeletalMesh_0}
   }
  }
  User.SpriteSize Array {
   type=NiagaraDataInterfaceArrayFloat
   defaultValue {
    struct=NiagaraExt_VariableValue_DataInterface
    value {
     dataInterface=&self:NiagaraDataInterfaceArrayFloat_0
     dataInterfaceClass=&NiagaraDataInterfaceArrayFloat
    }
   }
  }
 ]
}
